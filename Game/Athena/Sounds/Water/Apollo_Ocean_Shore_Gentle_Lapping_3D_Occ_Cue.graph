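# Sound cue for a gentle lapping water ambience along an ocean shore in Athena.
# A single looping wave player plays a lake shore laps loop, slightly boosted in volume and routed to the water ambient sound class.
# A very-close-range ocean shore occlusion attenuation asset shapes falloff, with attenuation and single-instance, stop-oldest concurrency override values also stored.
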SoundCue /Game/Athena/Sounds/Water/Apollo_Ocean_Shore_Gentle_Lapping_3D_Occ_Cue
root=WavePlayer_2
nodes [{id=WavePlayer_2, properties {bLooping=true, soundWaveAssetPtr=/Game/Athena/Sounds/Water/Ambient/Amb_Water_Lake_Shore_Laps_Loop_01}}]
properties {
 attenuationSettings=/Game/Athena/Sounds/Water/Attenuations/Apollo_Water_Ocean_Shore_Very_Close_Occ_Att
 maxDistance=2500
 soundClassObject=/Game/Sounds/FortSoundClassesAndModes/Fort_Water_Ambient
 volumeMultiplier=1.25
 attenuationOverrides {bSpatialize=false, falloffDistance=1500, nonSpatializedRadiusStart=700, stereoSpread=1000, attenuationShapeExtents {x=925}}
 concurrencyOverrides {maxCount=1, resolutionRule=StopOldest}
}
