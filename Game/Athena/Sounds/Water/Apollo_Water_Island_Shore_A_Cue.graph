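# Sound cue for looping island shore water ambience on the Apollo map, in the Fort_Water_Ambient sound class at reduced volume.
# A looping wave player reuses a medium, close water stream recording rather than a dedicated shore sound.
# It runs through an Attenuation node using the shared island shore attenuation asset, then into a root Mixer whose second input is empty.
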SoundCue /Game/Athena/Sounds/Water/Apollo_Water_Island_Shore_A_Cue
root=Mixer_0
nodes [
 {
  id=Attenuation_0
  children [WavePlayer_4]
  properties {attenuationSettings=/Game/Athena/Sounds/Water/Attenuations/Apollo_Water_Island_Shore_Att}
 }
 {id=Mixer_0, children [Attenuation_0, null], properties {inputVolume [1,1]}}
 {id=WavePlayer_4, properties {bLooping=true, soundWaveAssetPtr=/Game/Athena/Sounds/Water/Ambient/Amb_Water_Stream_Med_Close_Loop_02}}
]
properties {
 maxDistance=1712
 soundClassObject=/Game/Sounds/FortSoundClassesAndModes/Fort_Water_Ambient
 volumeMultiplier=0.6
 attenuationOverrides {
  bAttenuateWithLPF=true
  falloffDistance=6000
  lPFFrequencyAtMax=500
  nonSpatializedRadiusStart=700
  stereoSpread=1000
  attenuationShapeExtents {x=925}
 }
 concurrencyOverrides {maxCount=1, resolutionRule=PreventNew}
}
