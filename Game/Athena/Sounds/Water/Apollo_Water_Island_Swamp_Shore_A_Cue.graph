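# Looping ambient water cue, named for the Apollo island swamp shore, in the water ambient sound class.
# A mixer layers a close stream loop with one of two randomly chosen medium slurp water loops, borrowed from the Slurp item sounds.
# Each layer passes through its own attenuation node, using close and medium river settings respectively, so the blend varies with listener distance.
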SoundCue /Game/Athena/Sounds/Water/Apollo_Water_Island_Swamp_Shore_A_Cue
root=Mixer_0
nodes [
 {
  id=Attenuation_0
  children [WavePlayer_0]
  properties {attenuationSettings=/Game/Athena/Sounds/Water/Attenuations/Apollo_Water_River_Close_Att}
 }
 {id=Attenuation_2, children [Random_0], properties {attenuationSettings=/Game/Athena/Sounds/Water/Attenuations/Apollo_Water_River_Medium_Att}}
 {id=Mixer_0, children [Attenuation_0, Attenuation_2], properties {inputVolume [1,1]}}
 {id=Random_0, children [WavePlayer_2,WavePlayer_3], properties {weights [1,1]}}
 {id=WavePlayer_0, properties {bLooping=true, soundWaveAssetPtr=/Game/Athena/Sounds/Water/Ambient/Amb_Water_Stream_Med_Close_Loop_02}}
 {id=WavePlayer_2, properties {bLooping=true, soundWaveAssetPtr=/Game/Athena/Sounds/Items/Slurp/Item_Slurp_Water_Amb_Loop_Medium_01}}
 {id=WavePlayer_3, properties {bLooping=true, soundWaveAssetPtr=/Game/Athena/Sounds/Items/Slurp/Item_Slurp_Water_Amb_Loop_Medium_02}}
]
properties {
 attenuationSettings=&Apollo_Water_River_Close_Att
 maxDistance=6000
 soundClassObject=/Game/Sounds/FortSoundClassesAndModes/Fort_Water_Ambient
 volumeMultiplier=0.6
 attenuationOverrides {
  bAttenuateWithLPF=true
  falloffDistance=6000
  lPFFrequencyAtMax=500
  nonSpatializedRadiusStart=700
  stereoSpread=1000
  attenuationShapeExtents {x=925}
 }
 concurrencyOverrides {maxCount=1, resolutionRule=PreventNew}
}
