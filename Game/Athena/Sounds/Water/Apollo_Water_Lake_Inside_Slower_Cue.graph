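# Ambient water SoundCue that loops a distant lake-shore recording, named as a slower, inside-lake variant.
# A single looping Wave Player feeds the output directly with boosted volume, routed to the ambient water sound class.
# It references the shared close-river attenuation asset despite being a lake sound.
# It also stores non-spatialized attenuation and single-voice, stop-oldest concurrency values, but neither override is switched on.
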SoundCue /Game/Athena/Sounds/Water/Apollo_Water_Lake_Inside_Slower_Cue
root=WavePlayer_0
nodes [{id=WavePlayer_0, properties {bLooping=true, soundWaveAssetPtr=/Game/Athena/Sounds/Water/Ambient/Amb_Lake_Water_Shore_Dist_Loop}}]
properties {
 attenuationSettings=/Game/Athena/Sounds/Water/Attenuations/Apollo_Water_River_Close_Att
 maxDistance=1600
 soundClassObject=/Game/Sounds/FortSoundClassesAndModes/Fort_Water_Ambient
 volumeMultiplier=1.5
 attenuationOverrides {bSpatialize=false, falloffDistance=1500, nonSpatializedRadiusStart=700, stereoSpread=1000, attenuationShapeExtents {x=925}}
 concurrencyOverrides {maxCount=1, resolutionRule=StopOldest}
}
