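# Sound cue for a looping close-up lake shore water ambience, classed as water ambient with boosted volume.
# A looping wave player feeds a root attenuation node that borrows the close, focused river attenuation asset.
# The cue also stores attenuation override values with low-pass filtering and single-instance, prevent-new concurrency values, though neither override is shown as enabled.
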SoundCue /Game/Athena/Sounds/Water/Apollo_Water_Lake_Shore_B_Cue
root=Attenuation_1
nodes [
 {
  id=Attenuation_1
  children [WavePlayer_0]
  properties {attenuationSettings=/Game/Athena/Sounds/Water/Attenuations/Apollo_Water_River_Close_Focused_Att}
 }
 {id=WavePlayer_0, properties {bLooping=true, soundWaveAssetPtr=/Game/Athena/Sounds/Water/Ambient/Amb_Lake_Water_Shore_Close_Loop_01}}
]
properties {
 maxDistance=4512
 soundClassObject=/Game/Sounds/FortSoundClassesAndModes/Fort_Water_Ambient
 volumeMultiplier=2.5
 attenuationOverrides {
  bAttenuateWithLPF=true
  falloffDistance=6000
  lPFFrequencyAtMax=500
  nonSpatializedRadiusStart=700
  stereoSpread=1000
  attenuationShapeExtents {x=925}
 }
 concurrencyOverrides {maxCount=1, resolutionRule=PreventNew}
}
