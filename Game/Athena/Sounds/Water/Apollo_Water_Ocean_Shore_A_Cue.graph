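# Looping ambient SoundCue for ocean shore water, in the Fort_Water_Ambient sound class.
# The Intensity parameter crossfades from a half-volume calm bed of lake laps and stream at 0 to full ocean surf at 1.
# The surf mixes very close, close, medium and distant wave loops, each with its own attenuation asset to layer by listener distance, while both calm loops share the river-close attenuation.
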SoundCue /Game/Athena/Sounds/Water/Apollo_Water_Ocean_Shore_A_Cue
root=ParamCrossFade_2
nodes [
 {
  id=Attenuation_4
  children [WavePlayer_6]
  properties {attenuationSettings=/Game/Athena/Sounds/Water/Attenuations/Apollo_Water_Ocean_Shore_Close_Att}
 }
 {
  id=Attenuation_5
  children [WavePlayer_3]
  properties {attenuationSettings=/Game/Athena/Sounds/Water/Attenuations/Apollo_Water_Ocean_Shore_Medium_Att}
 }
 {
  id=Attenuation_6
  children [WavePlayer_2]
  properties {attenuationSettings=/Game/Athena/Sounds/Water/Attenuations/Apollo_Water_Ocean_Shore_Distant_Att}
 }
 {
  id=Attenuation_7
  children [WavePlayer_7]
  properties {attenuationSettings=/Game/Athena/Sounds/Water/Attenuations/Apollo_Water_Ocean_Shore_Very_Close_Att}
 }
 {
  id=Attenuation_8
  children [WavePlayer_18]
  properties {attenuationSettings=/Game/Athena/Sounds/Water/Attenuations/Apollo_Water_River_Close_Att}
 }
 {
  id=Attenuation_9
  children [WavePlayer_8]
  properties {attenuationSettings=&Apollo_Water_River_Close_Att}
 }
 {id=Mixer_0, children [Attenuation_7, Attenuation_4, Attenuation_5, Attenuation_6], properties {inputVolume [1,1,1,1]}}
 {id=Mixer_2, children [Attenuation_8, Attenuation_9], properties {inputVolume [1,1]}}
 {
  id=ParamCrossFade_2
  children [Mixer_2,Mixer_0]
  properties {
   paramName=Intensity
   crossFadeInput [
    {fadeInDistanceEnd=0, fadeInDistanceStart=0, fadeOutDistanceEnd=1, fadeOutDistanceStart=0, volume=0.5}
    {fadeInDistanceEnd=1, fadeInDistanceStart=0, fadeOutDistanceEnd=100, fadeOutDistanceStart=1, volume=1}
   ]
  }
 }
 {id=WavePlayer_18, properties {bLooping=true, soundWaveAssetPtr=/Game/Athena/Sounds/Water/Ambient/Amb_Water_Lake_Shore_Laps_Loop_01}}
 {id=WavePlayer_2, properties {bLooping=true, soundWaveAssetPtr=/Game/Athena/Sounds/Water/Ambient/Amb_Water_Ocean_Waves_Distant_Loop_01}}
 {id=WavePlayer_3, properties {bLooping=true, soundWaveAssetPtr=/Game/Athena/Sounds/Water/Ambient/Amb_Water_Ocean_Waves_Med_Distant_Loop_01}}
 {id=WavePlayer_6, properties {bLooping=true, soundWaveAssetPtr=/Game/Athena/Sounds/Water/Ambient/Amb_Water_Ocean_Waves_Close_Loop_01}}
 {id=WavePlayer_7, properties {bLooping=true, soundWaveAssetPtr=/Game/Athena/Sounds/Water/Ambient/Amb_Water_Ocean_Shore_SmWaves_Loop_01}}
 {id=WavePlayer_8, properties {bLooping=true, soundWaveAssetPtr=/Game/Athena/Sounds/Water/Ambient/Amb_Water_Stream_Med_Close_Loop_02}}
]
properties {
 maxDistance=10000
 soundClassObject=/Game/Sounds/FortSoundClassesAndModes/Fort_Water_Ambient
 volumeMultiplier=1.5
 attenuationOverrides {
  bAttenuateWithLPF=true
  falloffDistance=6000
  lPFFrequencyAtMax=500
  nonSpatializedRadiusStart=700
  stereoSpread=1000
  attenuationShapeExtents {x=925}
 }
 concurrencyOverrides {maxCount=1, resolutionRule=PreventNew}
}
