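# Looping river shore ambience cue on the water ambient sound class, driven by Velocity and Intensity parameters.
# Mixes three layers, each with its own attenuation asset: close 2D, close and medium.
# Velocity crossfades each from calmer loops to stronger, boosted ones, with random rapids variants in 2D; the medium layer's calm loop fades out far more gradually.
# A continuous modulator scales the mix volume by Intensity.
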SoundCue /Game/Athena/Sounds/Water/Apollo_Water_River_Shore_A_Cue
root=ModulatorContinuous_0
nodes [
 {
  id=Attenuation_0
  children [ParamCrossFade_0]
  properties {attenuationSettings=/Game/Athena/Sounds/Water/Attenuations/Apollo_Water_River_Close_2D_Att}
 }
 {
  id=Attenuation_1
  children [ParamCrossFade_1]
  properties {attenuationSettings=/Game/Athena/Sounds/Water/Attenuations/Apollo_Water_River_Close_Att}
 }
 {
  id=Attenuation_2
  children [ParamCrossFade_2]
  properties {attenuationSettings=/Game/Athena/Sounds/Water/Attenuations/Apollo_Water_River_Medium_Att}
 }
 {id=Mixer_0, children [Attenuation_0, Attenuation_1, Attenuation_2], properties {inputVolume [1,1,1]}}
 {id=ModulatorContinuous_0, children [Mixer_0], properties {volumeModulationParams {parameterName=Intensity}}}
 {
  id=ParamCrossFade_0
  children [WavePlayer_18, Random_4]
  properties {
   paramName=Velocity
   crossFadeInput [
    {fadeInDistanceEnd=0, fadeInDistanceStart=0, fadeOutDistanceEnd=1, fadeOutDistanceStart=0.5, volume=1}
    {fadeInDistanceEnd=1, fadeInDistanceStart=0.5, fadeOutDistanceEnd=100, fadeOutDistanceStart=100, volume=10}
   ]
  }
 }
 {
  id=ParamCrossFade_1
  children [WavePlayer_4,WavePlayer_3]
  properties {
   paramName=Velocity
   crossFadeInput [
    {fadeInDistanceEnd=0, fadeInDistanceStart=0, fadeOutDistanceEnd=1, fadeOutDistanceStart=0.5, volume=1}
    {fadeInDistanceEnd=1, fadeInDistanceStart=0.5, fadeOutDistanceEnd=100, fadeOutDistanceStart=100, volume=10}
   ]
  }
 }
 {
  id=ParamCrossFade_2
  children [WavePlayer_11, WavePlayer_2]
  properties {
   paramName=Velocity
   crossFadeInput [
    {fadeInDistanceEnd=0, fadeInDistanceStart=0, fadeOutDistanceEnd=100, fadeOutDistanceStart=1, volume=1}
    {fadeInDistanceEnd=1, fadeInDistanceStart=0.5, fadeOutDistanceEnd=100, fadeOutDistanceStart=100, volume=5}
   ]
  }
 }
 {id=Random_4, children [WavePlayer_15, WavePlayer_16], properties {weights [1,1]}}
 {id=WavePlayer_11, properties {bLooping=true, soundWaveAssetPtr=/Game/Athena/Sounds/Water/Ambient/Amb_Water_River_Close_Loop_01}}
 {id=WavePlayer_15, properties {bLooping=true, soundWaveAssetPtr=/Game/Athena/Sounds/Water/Ambient/Amb_Water_River_Large_Shore_Rapids_Loop_01}}
 {id=WavePlayer_16, properties {bLooping=true, soundWaveAssetPtr=/Game/Athena/Sounds/Water/Ambient/Amb_Water_River_Large_Shore_Rapids_Loop_02}}
 {id=WavePlayer_18, properties {bLooping=true, soundWaveAssetPtr=/Game/Athena/Sounds/Water/Ambient/Amb_Water_Lake_Shore_Laps_Loop_01}}
 {id=WavePlayer_2, properties {bLooping=true, soundWaveAssetPtr=/Game/Athena/Sounds/Water/Ambient/Amb_Water_River_Large_Distant_Loop_01}}
 {id=WavePlayer_3, properties {bLooping=true, soundWaveAssetPtr=/Game/Athena/Sounds/Water/Ambient/Amb_Water_River_Large_Close_Loop_01}}
 {id=WavePlayer_4, properties {bLooping=true, soundWaveAssetPtr=/Game/Athena/Sounds/Water/Ambient/Amb_Water_Stream_Med_Close_Loop_02}}
]
properties {
 maxDistance=6000
 soundClassObject=/Game/Sounds/FortSoundClassesAndModes/Fort_Water_Ambient
 volumeMultiplier=0.25
 attenuationOverrides {
  bAttenuateWithLPF=true
  bEnableListenerFocus=true
  falloffDistance=6000
  lPFFrequencyAtMax=500
  nonSpatializedRadiusStart=700
  stereoSpread=1000
  attenuationShapeExtents {x=925}
 }
 concurrencyOverrides {maxCount=1, resolutionRule=PreventNew}
 concurrencySet [/Game/Sounds/Concurrency_Rules/Waterfall_StopFarthestOldest_Conc]
}
