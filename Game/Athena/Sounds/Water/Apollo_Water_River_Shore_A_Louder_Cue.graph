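# Looping ambient Sound Cue for river shore water, a boosted-volume "Louder" variant limited to one concurrent instance.
# Four layers with separate attenuations (close 2D, close, delayed half-volume close without occlusion, and medium) feed one mixer.
# Each layer crossfades on the Velocity parameter from calmer lap, stream or river loops to louder large-river loops, with the 2D layer randomizing between two rapids loops.
# An Intensity parameter continuously modulates the mixed volume.
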SoundCue /Game/Athena/Sounds/Water/Apollo_Water_River_Shore_A_Louder_Cue
root=ModulatorContinuous_1
nodes [
 {
  id=Attenuation_0
  children [ParamCrossFade_0]
  properties {attenuationSettings=/Game/Athena/Sounds/Water/Attenuations/Apollo_Water_River_Close_2D_Att}
 }
 {
  id=Attenuation_2
  children [ParamCrossFade_2]
  properties {attenuationSettings=/Game/Athena/Sounds/Water/Attenuations/Apollo_Water_River_Close_Att}
 }
 {
  id=Attenuation_3
  children [ParamCrossFade_3]
  properties {attenuationSettings=/Game/Athena/Sounds/Water/Attenuations/Apollo_Water_River_Medium_Att}
 }
 {
  id=Attenuation_4
  children [ParamCrossFade_4]
  properties {attenuationSettings=/Game/Athena/Sounds/Water/Attenuations/Apollo_Water_River_Close_Smaller_No_Occlussion_Att}
 }
 {id=Delay_0, children [Attenuation_4], properties {delayMax=0.5, delayMin=0.5}}
 {id=Mixer_0, children [Attenuation_0, Attenuation_2, Delay_0, Attenuation_3], properties {inputVolume [1,1,0.5,1]}}
 {id=ModulatorContinuous_1, children [Mixer_0], properties {volumeModulationParams {parameterName=Intensity}}}
 {
  id=ParamCrossFade_0
  children [WavePlayer_18, Random_4]
  properties {
   paramName=Velocity
   crossFadeInput [
    {fadeInDistanceEnd=0, fadeInDistanceStart=0, fadeOutDistanceEnd=1, fadeOutDistanceStart=0.5, volume=1}
    {fadeInDistanceEnd=1, fadeInDistanceStart=0.5, fadeOutDistanceEnd=100, fadeOutDistanceStart=100, volume=10}
   ]
  }
 }
 {
  id=ParamCrossFade_2
  children [WavePlayer_0,WavePlayer_1]
  properties {
   paramName=Velocity
   crossFadeInput [
    {fadeInDistanceEnd=0, fadeInDistanceStart=0, fadeOutDistanceEnd=1, fadeOutDistanceStart=0.5, volume=1}
    {fadeInDistanceEnd=1, fadeInDistanceStart=0.5, fadeOutDistanceEnd=100, fadeOutDistanceStart=100, volume=10}
   ]
  }
 }
 {
  id=ParamCrossFade_3
  children [WavePlayer_11, WavePlayer_2]
  properties {
   paramName=Velocity
   crossFadeInput [
    {fadeInDistanceEnd=0, fadeInDistanceStart=0, fadeOutDistanceEnd=100, fadeOutDistanceStart=1, volume=1}
    {fadeInDistanceEnd=1, fadeInDistanceStart=0.5, fadeOutDistanceEnd=100, fadeOutDistanceStart=100, volume=5}
   ]
  }
 }
 {
  id=ParamCrossFade_4
  children [WavePlayer_5,WavePlayer_6]
  properties {
   paramName=Velocity
   crossFadeInput [
    {fadeInDistanceEnd=0, fadeInDistanceStart=0, fadeOutDistanceEnd=1, fadeOutDistanceStart=0.5, volume=1}
    {fadeInDistanceEnd=1, fadeInDistanceStart=0.5, fadeOutDistanceEnd=100, fadeOutDistanceStart=100, volume=10}
   ]
  }
 }
 {id=Random_4, children [WavePlayer_15, WavePlayer_16], properties {weights [1,1]}}
 {id=WavePlayer_0, properties {bLooping=true, soundWaveAssetPtr=/Game/Athena/Sounds/Water/Ambient/Amb_Water_Stream_Med_Close_Loop_02}}
 {id=WavePlayer_1, properties {bLooping=true, soundWaveAssetPtr=/Game/Athena/Sounds/Water/Ambient/Amb_Water_River_Large_Close_Loop_01}}
 {id=WavePlayer_11, properties {bLooping=true, soundWaveAssetPtr=/Game/Athena/Sounds/Water/Ambient/Amb_Water_River_Close_Loop_01}}
 {id=WavePlayer_15, properties {bLooping=true, soundWaveAssetPtr=/Game/Athena/Sounds/Water/Ambient/Amb_Water_River_Large_Shore_Rapids_Loop_01}}
 {id=WavePlayer_16, properties {bLooping=true, soundWaveAssetPtr=/Game/Athena/Sounds/Water/Ambient/Amb_Water_River_Large_Shore_Rapids_Loop_02}}
 {id=WavePlayer_18, properties {bLooping=true, soundWaveAssetPtr=/Game/Athena/Sounds/Water/Ambient/Amb_Water_Lake_Shore_Laps_Loop_01}}
 {id=WavePlayer_2, properties {bLooping=true, soundWaveAssetPtr=/Game/Athena/Sounds/Water/Ambient/Amb_Water_River_Large_Distant_Loop_01}}
 {id=WavePlayer_5, properties {bLooping=true, soundWaveAssetPtr=&Amb_Water_Stream_Med_Close_Loop_02}}
 {id=WavePlayer_6, properties {bLooping=true, soundWaveAssetPtr=&Amb_Water_River_Large_Close_Loop_01}}
]
properties {
 maxDistance=6000
 soundClassObject=/Game/Sounds/FortSoundClassesAndModes/Fort_Water_Ambient
 volumeMultiplier=2
 attenuationOverrides {
  bAttenuateWithLPF=true
  bEnableListenerFocus=true
  falloffDistance=6000
  lPFFrequencyAtMax=500
  nonSpatializedRadiusStart=700
  stereoSpread=1000
  attenuationShapeExtents {x=925}
 }
 concurrencyOverrides {maxCount=1, resolutionRule=PreventNew}
}
