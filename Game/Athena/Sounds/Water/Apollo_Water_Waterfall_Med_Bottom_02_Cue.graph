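# Looping ambient Sound Cue, named for the bottom of a medium Apollo waterfall in Athena.
# A single Wave Player root loops a large river shore rapids recording, with no randomization or mixing.
# It plays at boosted volume through the water ambient sound class and borrows the shared medium river attenuation asset.
# Inline attenuation and concurrency values are stored, but their override switches stay at the default off.
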SoundCue /Game/Athena/Sounds/Water/Apollo_Water_Waterfall_Med_Bottom_02_Cue
root=WavePlayer_2
nodes [{id=WavePlayer_2, properties {bLooping=true, soundWaveAssetPtr=/Game/Athena/Sounds/Water/Ambient/Amb_Water_River_Large_Shore_Rapids_Loop_02}}]
properties {
 attenuationSettings=/Game/Athena/Sounds/Water/Attenuations/Apollo_Water_River_Medium_Att
 maxDistance=6000
 soundClassObject=/Game/Sounds/FortSoundClassesAndModes/Fort_Water_Ambient
 volumeMultiplier=2
 attenuationOverrides {
  bAttenuateWithLPF=true
  bEnableListenerFocus=true
  falloffDistance=6000
  lPFFrequencyAtMax=500
  nonSpatializedRadiusStart=700
  stereoSpread=1000
  attenuationShapeExtents {x=925}
 }
 concurrencyOverrides {maxCount=1, resolutionRule=PreventNew}
}
